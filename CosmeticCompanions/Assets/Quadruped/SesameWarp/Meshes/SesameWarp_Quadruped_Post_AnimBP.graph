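# Post-process AnimBP for the SesameWarp quadruped cosmetic companion, adding procedural motion to its accessory bones.
# The input pose passes through a PoseAsset blend, then weighted multi-bone constraints drive tag, knot and left/right suriken from head, neck, spine and shoulder-area bones.
# Four world-space AnimDynamics nodes add sprung, angle-limited sway to the knot tails, scaled by DynamicsAlpha.
# Initialization caches mannequin and hoodie-variant flags, which the AnimGraph never reads.
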AnimBlueprint /CosmeticCompanions/Assets/Quadruped/SesameWarp/Meshes/SesameWarp_Quadruped_Post_AnimBP
target_skeleton=&/SesameWarp_Quadruped_Skeleton
graphs [
 AnimGraph {nodes [
   n1 = Input Pose()
   n2 = Get PoseAsset() -> (PoseAsset:Pose Asset Object Reference)
   n3 = "Constraint - tag (PPPPPPPP)"!(
    ComponentPose=Local To Component(LocalPose=Pose!(PoseAsset:Pose Asset Object Reference=@n2, SourcePose:Pose Link Structure=@n1) -> (Pose:Pose Link Structure) {is=title AnimGraphNode_PoseBlendNode, bIgnorePoseAssetSourceAnimationCurves=true, blendOption=Linear})
    Alpha=1
    ConstraintWeights_0=0.139
    ConstraintWeights_1:float=0.0445
    ConstraintWeights_2:float=0.0822
    ConstraintWeights_3:float=0.062
    ConstraintWeights_4:float=0.0599
    ConstraintWeights_5:float=0.04
    ConstraintWeights_6:float=0.2447
    ConstraintWeights_7:float=0.3277
   ) -> (Pose) {
    boneToModify {boneName=tag}
    constraintSetup [
     {targetBone {boneName=head}}
     {targetBone {boneName=shoulder_fr_l}}
     {targetBone {boneName=spine_03}}
     {targetBone {boneName=scapula_fr_l}}
     {targetBone {boneName=scapula_fr_r}}
     {targetBone {boneName=shoulder_fr_r}}
     {targetBone {boneName=neck_02}}
     {targetBone {boneName=neck_01}}
    ]
    constraintWeights [1,1,1,1,1,1,1,1]
   }
   n4 = "Constraint - knot (PPPPPPPP)"!(
    ComponentPose=@n3
    Alpha=1
    ConstraintWeights_0=0.3048
    ConstraintWeights_1:float=0.0487
    ConstraintWeights_2:float=0.0121
    ConstraintWeights_3:float=0.0163
    ConstraintWeights_4:float=0.1059
    ConstraintWeights_5:float=0.2256
    ConstraintWeights_6:float=0.1654
    ConstraintWeights_7:float=0.1212
   ) -> (Pose) {
    boneToModify {boneName=knot}
    constraintSetup [
     {targetBone {boneName=reverse_jaw}}
     {targetBone {boneName=spine_03}}
     {targetBone {boneName=scapula_fr_l}}
     {targetBone {boneName=scapula_fr_r}}
     {targetBone {boneName=spine_02}}
     {targetBone {boneName=neck_02}}
     {targetBone {boneName=spine_01}}
     {targetBone {boneName=neck_01}}
    ]
    constraintWeights [1,1,1,1,1,1,1,1]
   }
   n5 = "Constraint - suriken_l (PPPPPPPPP)"!(
    ComponentPose=@n4
    Alpha=1
    ConstraintWeights_0=0.0527
    ConstraintWeights_1:float=0.0011
    ConstraintWeights_2:float=0.0734
    ConstraintWeights_3:float=0.2118
    ConstraintWeights_4:float=0
    ConstraintWeights_5:float=0.1939
    ConstraintWeights_6:float=0.0394
    ConstraintWeights_7:float=0.3784
    ConstraintWeights_8:float=0.0493
   ) -> (Pose) {
    boneToModify {boneName=suriken_l}
    constraintSetup [
     {targetBone {boneName=shoulder_fr_l}}
     {targetBone {boneName=reverse_jaw}}
     {targetBone {boneName=spine_03}}
     {targetBone {boneName=scapula_fr_l}}
     {targetBone {boneName=cheek_bt_l}}
     {targetBone {boneName=spine_02}}
     {targetBone {boneName=neck_02}}
     {targetBone {boneName=spine_01}}
     {targetBone {boneName=neck_01}}
    ]
    constraintWeights [1,1,1,1,1,1,1,1,1]
   }
   n6 = "Constraint - suriken_r (PPPPPPPPP)"!(
    ComponentPose=@n5
    Alpha=1
    ConstraintWeights_0=0.0011
    ConstraintWeights_1:float=0.0734
    ConstraintWeights_2:float=0.2118
    ConstraintWeights_3:float=0
    ConstraintWeights_4:float=0.1939
    ConstraintWeights_5:float=0.0527
    ConstraintWeights_6:float=0.0395
    ConstraintWeights_7:float=0.3784
    ConstraintWeights_8:float=0.0493
   ) -> (Pose) {
    boneToModify {boneName=suriken_r}
    constraintSetup [
     {targetBone {boneName=reverse_jaw}}
     {targetBone {boneName=spine_03}}
     {targetBone {boneName=scapula_fr_r}}
     {targetBone {boneName=cheek_bt_r}}
     {targetBone {boneName=spine_02}}
     {targetBone {boneName=shoulder_fr_r}}
     {targetBone {boneName=neck_02}}
     {targetBone {boneName=spine_01}}
     {targetBone {boneName=neck_01}}
    ]
    constraintWeights [1,1,1,1,1,1,1,1,1]
   }
   n7 = Get DynamicsAlpha() -> (DynamicsAlpha:float)
   n8 = Anim Dynamics - Bone: knot_l_01(ComponentPose=@n6, Alpha=@n7, ExternalForce="0,0,0") {
    angularDampingOverride=0.94
    angularSpringConstant=5000
    bAngularSpring=true
    bOverrideAngularDamping=true
    bUseGravityOverride=true
    gravityScale=0.35
    simulationSpace=World
    alphaBoolBlend {blendOption=Linear}
    boundBone {boneName=knot_l_01}
    componentAppliedLinearAccClamp=<500 500 500>
    componentLinearAccScale=<1 1 1>
    componentLinearVelScale=<1 1 1>
    gravityOverride {z=-490}
    physicsBodyDefinitions [{
      boundBone {boneName=knot_l_01}
      boxExtents=<25 25 25>
      constraintSetup {angularLimitsMax=<5 10 30>, angularLimitsMin=<-5 -10 -30>, angularTarget=<1 0 0>}
      localJointOffset=<0 0 4>
     }]
   }
   n9 = Anim Dynamics - Bone: knot_l_02(ComponentPose=@n8, Alpha=@n7, ExternalForce="0,0,0") {
    angularDampingOverride=0.94
    angularSpringConstant=5000
    bAngularSpring=true
    bOverrideAngularDamping=true
    bUseGravityOverride=true
    gravityScale=0.35
    simulationSpace=World
    alphaBoolBlend {blendOption=Linear}
    boundBone {boneName=knot_l_02}
    componentAppliedLinearAccClamp=<500 500 500>
    componentLinearAccScale=<1 1 1>
    componentLinearVelScale=<1 1 1>
    gravityOverride {z=-490}
    physicsBodyDefinitions [{
      boundBone {boneName=knot_l_02}
      boxExtents=<20 20 20>
      constraintSetup {angularLimitsMax=<5 10 30>, angularLimitsMin=<-5 -10 -30>, angularTarget=<1 0 0>}
      localJointOffset=<0 0 4>
     }]
   }
   n10 = Anim Dynamics - Bone: knot_r_01(ComponentPose=@n9, Alpha=@n7, ExternalForce="0,0,0") {
    angularDampingOverride=0.94
    angularSpringConstant=5000
    bAngularSpring=true
    bOverrideAngularDamping=true
    bUseGravityOverride=true
    gravityScale=0.35
    simulationSpace=World
    alphaBoolBlend {blendOption=Linear}
    boundBone {boneName=knot_r_01}
    componentAppliedLinearAccClamp=<500 500 500>
    componentLinearAccScale=<1 1 1>
    componentLinearVelScale=<1 1 1>
    gravityOverride {z=-490}
    physicsBodyDefinitions [{
      boundBone {boneName=knot_r_01}
      boxExtents=<25 25 25>
      constraintSetup {angularLimitsMax=<5 10 30>, angularLimitsMin=<-5 -10 -30>, angularTarget=<1 0 0>}
      localJointOffset=<0 0 4>
     }]
   }
   n11 = Anim Dynamics - Bone: knot_r_02(ComponentPose=@n10, Alpha=@n7, ExternalForce="0,0,0") {
    angularDampingOverride=0.94
    angularSpringConstant=5000
    bAngularSpring=true
    bOverrideAngularDamping=true
    bUseGravityOverride=true
    gravityScale=0.35
    simulationSpace=World
    alphaBoolBlend {blendOption=Linear}
    boundBone {boneName=knot_r_02}
    componentAppliedLinearAccClamp=<500 500 500>
    componentLinearAccScale=<1 1 1>
    componentLinearVelScale=<1 1 1>
    gravityOverride {z=-490}
    physicsBodyDefinitions [{
      boundBone {boneName=knot_r_02}
      boxExtents=<20 20 20>
      constraintSetup {angularLimitsMax=<5 10 30>, angularLimitsMin=<-5 -10 -30>, angularTarget=<1 0 0>}
      localJointOffset=<0 0 4>
     }]
   }
   Output Pose(Result=Component To Local(ComponentPose=@n11))
  ]}
 EventGraph {nodes [
   Event BlueprintInitializeAnimation()
   > Set isCompanionMannequin(isCompanionMannequin:bool=IsA!(Object:Object Reference=GetOwningActor(), ObjectClass:Object Class Reference=/Script/CosmeticCompanionsRuntime.FortCompanionMannequin) -> (ReturnValue:bool)) -> (Output_Get:bool)
   > Set hasAHoodie(hasAHoodie:bool=HasVariantTagForChannel!(
     self:Fort Player Companion Anim Instance Object Reference=GetMainAnimInstance!(self:Companion Post Process Anim Instance Object Reference) -> (ReturnValue:Fort Player Companion Anim Instance Object Reference)
     ChannelTag:Gameplay Tag Structure
     VariantTag:Gameplay Tag Structure
    ) -> (ReturnValue:bool)) -> (Output_Get:bool)
  ]}
]
variables [
 hasAHoodie {type=boolean}
 isCompanionMannequin {type=boolean}
 PoseAsset {type=PoseAsset, default=&/SesameWarp_Quadruped_PoseAsset}
 TailRotation {type=Rotator}
]
